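; BGA_TestMap_Cheater is an interactable test-map actor that runs a configurable list of console commands for the player who uses it.
; On interact, it resolves and saves the pawn's Fort player controller, then RunCheats executes each CheatsToRun entry through that controller.
; Interaction is always allowed, with a configurable prompt.
; The construction script applies the display text and interaction speed, and sets the cylinder material's Color parameter.
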(event Custom|RunCheats
  ;[1 1 1] Run Cheats - Execute each console command
  (for _array_element (Variables|Default|GetCheatsToRun)
    (Development|ExecuteConsoleCommand _array_element (Variables|Default|GetInteractingController))))

(event Interaction|EventBlueprintOnInteract
  (params (InteractingPawn :type Fort%20Pawn%20Object%20Reference) (InteractionBeingAttempted :type EInteractionBeingAttempted%20Enum :default "FirstInteraction"))
  (bind _returnvalue (Fortnite|GetFortPlayerControllerfromActor InteractingPawn))
  ;[1 1 1] Event On Interact - If the player controller is valid save it and run the cheat array
  (Utilities|IsValid _returnvalue
    (:Is%20Valid
      (Variables|Default|SetInteractingController _returnvalue)
      (CallFunction|RunCheats))))

(fn BlueprintCanInteract
  (params (InteractingPawn :type Fort%20Pawn%20Object%20Reference) (InteractionBeingAttempted :type EInteractionBeingAttempted%20Enum :default "FirstInteraction") (InteractionType :type TInteractionType%20Enum :default "IT_NoInteraction"))
  (return true))

(fn BlueprintGetInteractionString
  (params (InteractingPawn :type Fort%20Pawn%20Object%20Reference) (InteractionBeingAttempted :type EInteractionBeingAttempted%20Enum :default "FirstInteraction"))
  (return (Variables|Text|GetInteractText)))

(fn BlueprintGetInteractionTime
  (params (InteractingPawn :type Fort%20Pawn%20Object%20Reference) (InteractionBeingAttempted :type EInteractionBeingAttempted%20Enum :default "FirstInteraction"))
  (return 0.0 true))

(fn ConstructionScript
  (Rendering|Components|TextRender|SetText (Variables|Default|GetTextEdit) (Variables|Text|GetTextToDisplay))
  (Variables|Interaction|SetInteractionSpeed (Variables|Text|GetInteractSpeed))
  (Rendering|Material|SetVectorParameterValueonMaterials (Variables|Default|GetCylinder) "Color" "0, 0, 0"))
