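; Test-map interactable building actor (BGA) that toggles a configured Fort ability set on each pawn that interacts with it.
; Interaction checks the InteractingPawns array, then equips or unequips the set, tracking handles in a parallel array matched by index.
; RepUpdateBGAState switches the cylinder visibility, apply/remove text and mesh color only for the interacting player's local controller.
; The construction script sets the ability-set label, interaction speed and cylinder color.
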(event Interaction|EventBlueprintOnInteract
  (params (InteractingPawn :type Fort%20Pawn%20Object%20Reference) (InteractionBeingAttempted :type EInteractionBeingAttempted%20Enum :default "FirstInteraction"))
  (bind _interactingpawn (Variables|Default|GetInteractingPawn))
  (bind _interactingpawns (Variables|Default|GetInteractingPawns))
  (bind _returnvalue (Utilities|Array|ContainsItem _interactingpawns _interactingpawn))
  (bind _interactingpawn_2 (Variables|Default|GetInteractingPawn))
  (bind _interactingpawns_2 (Variables|Default|GetInteractingPawns))
  (bind _returnvalue_2 (Utilities|Array|FindItem _interactingpawns_2 _interactingpawn_2))
  (bind _abilitysethandles (Variables|Default|GetAbilitySetHandles))
  (bind _output (Utilities|Array|Get(aref) _abilitysethandles _returnvalue_2))
  ;[1 1 1] Apply or Remove GAS from interacting player
  (Variables|Default|SetInteractingPawn InteractingPawn)
  (CallFunction|RepUpdateBGAState (Fortnite|GetFortPlayerControllerfromActor InteractingPawn) _returnvalue)
  (if _returnvalue
    (Development|PrintString "Remove Ability Set" true true "(R=1.000000,G=0.817612,B=0.000000,A=1.000000)")
    (Ability|UnequipFortAbilitySet _output)
    (Utilities|Array|RemoveItem (Variables|Default|GetAbilitySetHandles) _output)
    (Utilities|Array|RemoveItem (Variables|Default|GetInteractingPawns) (Variables|Default|GetInteractingPawn))
    (else
      (Development|PrintString "Apply Ability Set" true true "(R=1.000000,G=0.817612,B=0.000000,A=1.000000)")
      (Utilities|Array|AddUnique (Variables|Default|GetInteractingPawns) (Variables|Default|GetInteractingPawn))
      (bind _returnvalue_3 (Ability|EquipFortAbilitySet (Variables|Default|GetInteractingPawn) (Variables|Default|GetAbilitySetToApply)))
      (Utilities|Array|AddUnique (Variables|Default|GetAbilitySetHandles) _returnvalue_3))))

(event Custom|RepUpdateBGAState
  (params (PawnController :type Fort%20Player%20Controller%20Object%20Reference) (Remove :type Boolean))
  ;[1 1 1] Update visual state of BGA for interacting player
  (Utilities|IsValid PawnController
    (:Is%20Valid
      (if (Pawn|IsLocalController PawnController)
        (Development|PrintString "Update Local")
        (if Remove
          (Development|SetHiddeninGame (Variables|Default|GetCylinder) false false)
          (Rendering|Components|TextRender|SetText (Variables|Default|GetTextApplyRemove) (Variables|Default|GetApplyText))
          (Rendering|Material|SetVectorParameterValueonMaterials (Variables|Default|GetStaticMesh) "Color" "0.000000,256.000000, 0")
          (else
            (Development|SetHiddeninGame (Variables|Default|GetCylinder) true false)
            (Rendering|Components|TextRender|SetText (Variables|Default|GetTextApplyRemove) (Variables|Default|GetRemoveText))
            (Rendering|Material|SetVectorParameterValueonMaterials (Variables|Default|GetStaticMesh) "Color" "256.000000,0.000000, 0")))
        (else
          _)))
    (:Is%20Not%20Valid
      _))
  (Development|PrintString "Don't Update Non-Local"))

(fn BlueprintCanInteract
  (params (InteractingPawn :type Fort%20Pawn%20Object%20Reference) (InteractionBeingAttempted :type EInteractionBeingAttempted%20Enum :default "FirstInteraction") (InteractionType :type TInteractionType%20Enum :default "IT_NoInteraction"))
  (return true))

(fn BlueprintGetInteractionString
  (params (InteractingPawn :type Fort%20Pawn%20Object%20Reference) (InteractionBeingAttempted :type EInteractionBeingAttempted%20Enum :default "FirstInteraction"))
  (return (Variables|Text|GetInteractText)))

(fn ConstructionScript
  (Rendering|Components|TextRender|SetText (Variables|Default|GetTextAbilitySet) (Variables|Text|GetAbilitySetTextName))
  (Variables|Interaction|SetInteractionSpeed (Variables|Default|GetInteractSpeed))
  (Rendering|Material|SetVectorParameterValueonMaterials (Variables|Default|GetCylinder) "Color" "0,256.000000, 0"))
