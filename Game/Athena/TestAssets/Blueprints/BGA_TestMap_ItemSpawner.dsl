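; Interactable test-map actor that spawns Fortnite item pickups in a row along its forward vector and replaces each when picked up.
; Interacting spawns its hard-referenced (deprecated) and blocking-loaded soft-referenced items, then runs a console cheat.
; A master spawner instead triggers every spawner in its array, with short delays to limit hitching.
; Each pickup's OnPickup event respawns that item in place, after an optional retriggerable delay.
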(event EventBeginPlay
  (for _array_element (Variables|Spawning|GetArrayToSpawn_(_DEPRECATED_))
    (Utilities|Array|AddUnique (Variables|Default|GetPrimary%20Asset%20List) (AssetManager|GetPrimaryAssetIdfromObject _array_element)))
  (AssetManager|AsyncChangeBundleStateForPrimaryAssetList (Variables|Default|GetPrimary%20Asset%20List) (Variables|Default|GetItemAssetNames) (Utilities|Array|MakeArray)))

(event Custom|Spawn%20Items
  (params (Items :type Array%20of%20Fort%20World%20Item%20Definition%20Object%20References) (Items%20%28Soft%20Refs%29 :type Array%20of%20Fort%20World%20Item%20Definition%20Soft%20Object%20References) (StartLoc :type Vector) (NoOffset :type Boolean))
  (bind _itemstospawn (Variables|Default|GetItemsToSpawn))
  ;[1 1 1] Spawn items from the array
  (Variables|Default|SetItemsToSpawn Items)
  (bind _asfort_world_item_definition (CallFunction|AssembleItemIDArrayfromSoftReferences Items%20%28Soft%20Refs%29))
  (Utilities|Array|AppendArray _itemstospawn _asfort_world_item_definition)
  (bind _spawneditems (CallFunction|SpawnonServer _itemstospawn StartLoc NoOffset))
  (for _array_element _spawneditems
    (Interaction|BindEventtoOnPickup _array_element (delegate Custom|Respawn%20Items))))

(event Interaction|EventBlueprintOnInteract
  (params (InteractingPawn :type Fort%20Pawn%20Object%20Reference) (InteractionBeingAttempted :type EInteractionBeingAttempted%20Enum :default "FirstInteraction"))
  ;[1 1 1] On Interact spawn items for self or others if a master.
  (if (Variables|MasterSpawner|GetMasterSpawner)
    (CallFunction|SpawnAll)
    (else
      (CallFunction|SpawnItems (Variables|Spawning|GetArrayToSpawn_(_DEPRECATED_)) (Variables|Spawning|GetArrayToSpawnSoftReference) (+ (Transformation|GetActorLocation) (Variables|Spawning|GetSpawnOffset)))
      (Development|ExecuteConsoleCommand (Variables|Default|Getcheat%20to%20execute)))))

(event Custom|Respawn%20Items
  (params (SelfActor :type Fort%20Pickup%20Object%20Reference) (InteractingPawn :type Fort%20Pawn%20Object%20Reference))
  (bind _itemrespawndelay (Variables|Spawning|GetItemRespawnDelay))
  ;[1 1 1] Spawn Item when it is picked up to replace it.
  (if (> _itemrespawndelay 0)
    (Utilities|FlowControl|RetriggerableDelay _itemrespawndelay)
    (else
      _))
  (CallFunction|SpawnItems :Items (Utilities|Array|MakeArray (Utilities|Casting|CastToFortWorldItemDefinition (Pickup|GetPrimaryItemDefinition SelfActor))) :StartLoc (Class|FortPickup|GetPickupFinalLocation SelfActor) :NoOffset true))

(event Custom|SpawnAll
  ;[1 1 1] Call Spawn On all items in the master spawner array
  (for _array_element (Variables|MasterSpawner|GetMasterSpawnerArray)
    (CallFunction|SpawnItems _array_element (Variables|Spawning|GetArrayToSpawn_(_DEPRECATED_) _array_element) (Variables|Spawning|GetArrayToSpawnSoftReference) (+ (Transformation|GetActorLocation _array_element) (Variables|Spawning|GetSpawnOffset)) false)
    (Utilities|FlowControl|Delay 0.1) ; Delay So that it does not hitch super hard
  ))

(fn AssembleItemIDArrayFromSoftReferences
  (params (Array :type Array%20of%20Fort%20World%20Item%20Definition%20Soft%20Object%20References))
  (for _array_element Array
    (bind _returnvalue (Utilities|LoadAssetBlocking _array_element))
    (bind _asfort_world_item_definition (Utilities|Casting|CastToFortWorldItemDefinition _returnvalue))
    (Utilities|Array|AddUnique (LocalVariables|GetFortWorldItemIDs) _asfort_world_item_definition))
  (return (LocalVariables|GetFortWorldItemIDs)))

(fn BlueprintCanInteract
  (params (InteractingPawn :type Fort%20Pawn%20Object%20Reference) (InteractionBeingAttempted :type EInteractionBeingAttempted%20Enum :default "FirstInteraction") (InteractionType :type TInteractionType%20Enum :default "IT_NoInteraction"))
  (return true))

(fn BlueprintGetInteractionString
  (params (InteractingPawn :type Fort%20Pawn%20Object%20Reference) (InteractionBeingAttempted :type EInteractionBeingAttempted%20Enum :default "FirstInteraction"))
  (return (Variables|Text|GetInteractText)))

(fn BlueprintGetInteractionTime
  (params (InteractingPawn :type Fort%20Pawn%20Object%20Reference) (InteractionBeingAttempted :type EInteractionBeingAttempted%20Enum :default "FirstInteraction"))
  (return 0.0 true))

(fn SpawnOnServer
  (params (Array :type Array%20of%20Fort%20World%20Item%20Definition%20Object%20References) (StartLoc :type Vector) (NoOffset :type Boolean))
  (for (_array_element _array_index) Array
    (bind _returnvalue (Inventory|SpawnItemVariantPickupinWorld (Utilities|Struct|MakeSpawnItemVariantParams _array_element "(GameplayTags=)" 1 (+ (+ StartLoc (select NoOffset "0, 0, 0" (* (Transformation|GetActorForwardVector) (Variables|Default|GetForwardOffset)))) (* (Math|Vector|Normalize (Transformation|GetActorForwardVector)) (* _array_index (Variables|Spawning|GetWeaponPlaceOffset)))) "0.000000,0.000000,1.000000" 0 false false (Variables|Default|GetBlockedFromAutoPickup) -1 "Tossed" "Unset" 0 0 false)))
    (Utilities|IsValid _returnvalue
      (:Is%20Valid
        (Utilities|Array|AddUnique (LocalVariables|GetSpawnedItemsLocal) _returnvalue))))
  (return (LocalVariables|GetSpawnedItemsLocal)))

(fn ConstructionScript
  (Rendering|Components|TextRender|SetText (Variables|Default|GetTextEdit) (Variables|Text|GetTextToDisplay))
  (Variables|Interaction|SetInteractionSpeed (Variables|Text|GetInteractSpeed))
  (Rendering|Material|SetVectorParameterValueonMaterials (Variables|Default|GetCylinder) "Color" "0, 0, 0")
  (if (Variables|MasterSpawner|GetMasterSpawner)
    (Development|SetHiddeninGame (Variables|Default|GetMasterHat) false true)))
