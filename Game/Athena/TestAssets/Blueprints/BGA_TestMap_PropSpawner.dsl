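; Interactable test-map actor that spawns a configured array of building static mesh props when a pawn interacts with it.
; Interaction is always allowed, and Spawn Items places each prop along the actor's forward vector from its location, at a fixed lead distance plus an index-scaled offset.
; The construction script applies display text, interaction speed and cylinder color.
; The MasterSpawner flag only unhides a hat component; interaction spawns just this actor's own array.
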(event Custom|Spawn%20Items
  (params (Items :type Array%20of%20Building%20Static%20Mesh%20Class%20References) (StartLoc :type Vector))
  ;[1 1 1] Spawn items from the array
  (for (_array_element _array_index) Items
    (Building|SpawnBuildingActor _array_element (Math|Conversions|ToTransform(Vector) (+ StartLoc (* (Math|Vector|Normalize (Transformation|GetActorForwardVector)) (Math|Conversions|ToFloat(Integer) (+ (Math|Float|Truncate (* _array_index (Variables|Spawning|GetWeaponPlaceOffset))) 100))))))))

(event Interaction|EventBlueprintOnInteract
  (params (InteractingPawn :type Fort%20Pawn%20Object%20Reference) (InteractionBeingAttempted :type EInteractionBeingAttempted%20Enum :default "FirstInteraction"))
  ;[1 1 1] On Interact spawn items for self or others if a master.
  (CallFunction|SpawnItems (Variables|Spawning|GetArrayToSpawn) (Transformation|GetActorLocation)))

(fn BlueprintCanInteract
  (params (InteractingPawn :type Fort%20Pawn%20Object%20Reference) (InteractionBeingAttempted :type EInteractionBeingAttempted%20Enum :default "FirstInteraction") (InteractionType :type TInteractionType%20Enum :default "IT_NoInteraction"))
  (return true))

(fn BlueprintGetInteractionString
  (params (InteractingPawn :type Fort%20Pawn%20Object%20Reference) (InteractionBeingAttempted :type EInteractionBeingAttempted%20Enum :default "FirstInteraction"))
  (return (Variables|Text|GetInteractText)))

(fn ConstructionScript
  (Rendering|Components|TextRender|SetText (Variables|Default|GetTextEdit) (Variables|Text|GetTextToDisplay))
  (Variables|Interaction|SetInteractionSpeed (Variables|Text|GetInteractSpeed))
  (Rendering|Material|SetVectorParameterValueonMaterials (Variables|Default|GetCylinder) "Color" "0, 0, 0")
  (if (Variables|MasterSpawner|GetMasterSpawner)
    (Development|SetHiddeninGame (Variables|Default|GetMasterHat) false true)))
