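; Interactable test-map actor that, per its configuration, grants a stat buff or debuff to the interacting pawn or removes such effects.
; Granting applies a gameplay effect spec whose SetByCaller.Stats magnitude is one plus or minus StatModAmount, then prints the stacked percentage.
; Remover mode strips all stacks of each listed effect from the pawn.
; Construction sets interaction speed, mesh visibility and mode label; BeginPlay sets the display text.
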(event Interaction|EventBlueprintOnInteract
  (params (InteractingPawn :type Fort%20Pawn%20Object%20Reference) (InteractionBeingAttempted :type EInteractionBeingAttempted%20Enum :default "FirstInteraction"))
  (if (Variables|Default|GetRemover)
    (Development|PrintString "Removed Stat Buffs" true true "(R=1.000000,G=0.817612,B=0.000000,A=1.000000)")
    (for _array_element (Variables|Default|GetGE_ToRemove)
      (GameplayEffects|RemoveActiveGameplayEffectbySourceEffect (Ability|GetAbilitySystemComponent InteractingPawn) _array_element 0 (GameplayEffects|GetGameplayEffectCount (Ability|GetAbilitySystemComponent InteractingPawn) _array_element 0 true)))
    (else
      (bind _returnvalue (Ability|GameplayEffect|AssignTagSetbyCallerMagnitude (GameplayEffects|MakeOutgoingSpec :self (Ability|GetAbilitySystemComponent self) :GameplayEffectClass (Variables|Default|GetGE_ToApply) :Level 0.0) "(TagName=\"SetByCaller.Stats\")" (Math|Float|SelectFloat (+ (* (Variables|Default|GetStatModAmount) -1.0) 1.0) (+ (Variables|Default|GetStatModAmount) 1.0) (Variables|Default|GetDebuff))))
      (bind _returnvalue_2 (GameplayEffects|ApplyGameplayEffectSpecToTarget (Ability|GetAbilitySystemComponent self) _returnvalue (Ability|GetAbilitySystemComponent InteractingPawn)))
      (bind _returnvalue_3 (Ability|GameplayEffect|GetActiveGameplayEffectStackCount _returnvalue_2))
      (Development|PrintString (Utilities|String|Append (Utilities|String|Append (Utilities|String|Append (select (Variables|Default|GetDebuff) "Added Debuff: " "Added Buff: ") (Utilities|GetClassDisplayName (Variables|Default|GetGE_ToApply))) (Utilities|String|Append (Utilities|String|Append " = " (select (Variables|Default|GetDebuff) "-" "+")) (Utilities|String|ToString(Float) (* (* _returnvalue_3 (Variables|Default|GetStatModAmount)) 1e+02)))) "%") true true "(R=1.000000,G=0.898993,B=0.000000,A=1.000000)" 3.0))))

(event EventBeginPlay
  (|CallParentFunction)
  (Rendering|Components|TextRender|SetText (Variables|Default|GetTextEdit) (Variables|Text|GetTextToDisplay)))

(fn BlueprintCanInteract
  (params (InteractingPawn :type Fort%20Pawn%20Object%20Reference) (InteractionBeingAttempted :type EInteractionBeingAttempted%20Enum :default "FirstInteraction") (InteractionType :type TInteractionType%20Enum :default "IT_NoInteraction"))
  (return true))

(fn BlueprintGetInteractionString
  (params (InteractingPawn :type Fort%20Pawn%20Object%20Reference) (InteractionBeingAttempted :type EInteractionBeingAttempted%20Enum :default "FirstInteraction"))
  (return (Variables|Text|GetInteractText)))

(fn ConstructionScript
  (Variables|Interaction|SetInteractionSpeed (Variables|Text|GetInteractSpeed))
  (Rendering|Material|SetVectorParameterValueonMaterials (Variables|Default|GetCylinder) "Color" "0, 0, 0")
  (if (Variables|Default|GetRemover)
    (Development|SetHiddeninGame (Variables|Default|GetCylinder) true false)
    (Rendering|Components|TextRender|SetText (Variables|Default|GetTextDefault) (Variables|Default|GetRemovalText))
    (Development|SetHiddeninGame (Variables|Default|GetMasterHat) false true)
    (elif (Variables|Default|GetDebuff)
      (Development|SetHiddeninGame (Variables|Default|GetCylinder) true false)
      (Rendering|Components|TextRender|SetText (Variables|Default|GetTextDefault) (Variables|Default|GetDebuffText)))))
