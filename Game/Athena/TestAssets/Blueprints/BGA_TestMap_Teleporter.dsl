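; BGA_TestMap_Teleporter is an interactable test-map actor that teleports the pawn using it to a set destination.
; The construction script sets the floating text, moves the SpawnLocation component to the configured TeleportLocation, applies the interaction speed and sets the Cone material color to black.
; Interaction is always allowed and shows a configurable prompt; on interact, the pawn is moved to SpawnLocation's world position without sweep.
; BeginPlay, overlap and Tick events are empty.
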(event EventBeginPlay)

(event Collision|EventActorBeginOverlap
  (params (OtherActor :type Actor%20Object%20Reference)))

(event EventTick
  (params (DeltaSeconds :type Float%20%28single-precision%29)))

(event Interaction|EventBlueprintOnInteract
  (params (InteractingPawn :type Fort%20Pawn%20Object%20Reference) (InteractionBeingAttempted :type EInteractionBeingAttempted%20Enum :default "FirstInteraction"))
  (Transformation|SetActorLocation InteractingPawn (Transformation|GetWorldLocation (Variables|Default|GetSpawnLocation)) false false))

(fn BlueprintCanInteract
  (params (InteractingPawn :type Fort%20Pawn%20Object%20Reference) (InteractionBeingAttempted :type EInteractionBeingAttempted%20Enum :default "FirstInteraction") (InteractionType :type TInteractionType%20Enum :default "IT_NoInteraction"))
  (return true))

(fn BlueprintGetInteractionString
  (params (InteractingPawn :type Fort%20Pawn%20Object%20Reference) (InteractionBeingAttempted :type EInteractionBeingAttempted%20Enum :default "FirstInteraction"))
  (return (Variables|Default|GetInteractText)))

(fn ConstructionScript
  (Rendering|Components|TextRender|SetText (Variables|Default|GetTextEdit) (Variables|Default|GetTextToDisplay))
  (Transformation|SetWorldLocation (Variables|Default|GetSpawnLocation) (Variables|Default|GetTeleportLocation) false false)
  (Variables|Interaction|SetInteractionSpeed (Variables|Default|GetInteractSpeed))
  (Rendering|Material|SetVectorParameterValueonMaterials (Variables|Default|GetCone) "Color" "0, 0, 0"))
